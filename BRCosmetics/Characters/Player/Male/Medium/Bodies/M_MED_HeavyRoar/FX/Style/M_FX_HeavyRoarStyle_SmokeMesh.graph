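# Translucent unlit, two-sided smoke material for the HeavyRoar style character effect, built for Niagara mesh particles.
# World position offset combines SplineThicken strip widening, an axis rotation, a directional sine wobble and dynamic-parameter drag, with wobble and drag weighted along UV V.
# Particle color tints a panning smoke-trail texture, desynchronized by vertex color and particle random.
# Depth fade softens opacity except at Low and Medium quality.
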Material /BRCosmetics/Characters/Player/Male/Medium/Bodies/M_MED_HeavyRoar/FX/Style/M_FX_HeavyRoarStyle_SmokeMesh
graph {
 properties {
  bUsedWithInstancedStaticMeshes=true
  bUsedWithNanite=true
  bUsedWithNiagaraMeshParticles=true
  blendMode=BLEND_Translucent
  pixelDepthOffsetMode=PDOM_Legacy
  shadingModel=MSM_Unlit
  twoSided=true
  usedShadingModels=Unlit
 }
 nodes [
  n1 = ScalarParameter() {defaultValue=5, parameterName=ThicknessBase}
  n2 = ScalarParameter() {defaultValue=5, parameterName=ThicknessTip}
  n3 = Add(A=NamedRerouteUsage(), B=WorldPosition().XYZ)
  n4 = OneMinus(None=ComponentMask(None=TextureCoordinate()) {g=true})
  n5 = Add(A=ComponentMask(None=TextureCoordinate() {vTiling=0.25}) {g=true}, B=NamedRerouteUsage())
  n6 = Transform(None=VectorParameter() {parameterName=WobbleLocalDir, defaultValue=<rgba 0 0 0 1>}.RGB) {transformSourceType=TRANSFORMSOURCE_Instance}
  n7 = Multiply(A=Multiply(A=Sine(None=@n5) {period=2}, B=Normalize(VectorInput=@n6)), B=ScalarParameter() {parameterName=WPO_WiggleAmount})
  n8 = OneMinus(None=ComponentMask(None=TextureCoordinate()) {g=true})
  n9 = ScalarParameter() {defaultValue=1, parameterName=DragWeightExp}
  n10 = DynamicParameter() {paramNames [Param1,Param2,Param3,DragOffset]}
  n11 = Subtract(
   A=MaterialFunctionCall(R=@n10.Param1, G=@n10.Param1, B=@n10.Param1) {materialFunction=&AppendMany}.RGB
   B=ParticlePositionWS()
  )
  n12 = Multiply(A=@n10.DragOffset, B=ScalarParameter() {parameterName=DragOffsetScale})
  Add(
   A=Add(
    A=MaterialFunctionCall(WidthBase=@n1, WidthTip=@n2, WorldPosition=@n3) {materialFunction=&SplineThicken}.WorldPositionOffset
    B=Multiply(A=@n4, B=@n7)
   )
   B=Multiply(A=Power(Base=@n8, Exp=@n9), B=Multiply(A=@n11, B=@n12))
  )
  Add(B=NamedRerouteUsage())
  MaterialFunctionCall(Emissive=NamedRerouteUsage()) {materialFunction=&MF_FX_EmissiveAdjustments}
  n13 = ParticleColor()
  NamedRerouteDeclaration(None=@n13.RGB) {name=FinalColor, nodeColor=<rgba 0 0.7891 1 1>}
  n14 = Multiply(A=TextureCoordinate() {vTiling=0.25}, B=AppendVector(A=Constant() {r=1}, B=ScalarParameter() {defaultValue=1, parameterName=Tiling}))
  n15 = TextureSample(UVs=Panner(
    Coordinate=MaterialFunctionCall(XY=@n14) {materialFunction=&Swizzle}.YX
    Time=NamedRerouteUsage()
   ) {bFractionalPart=true, speedX=0.125}) {texture=/Game/Effects/Fort_Effects/Textures/Smoke/T_Smoke_Trail_Soft_2}
  n16 = TextureCoordinate() {vTiling=2}
  n17 = Saturate(None=Multiply(A=ComponentMask(None=@n16) {g=true}))
  NamedRerouteDeclaration(None=Multiply(A=@n13.A, B=Multiply(A=@n15.G, B=@n17))) {name=FinalOpacity, nodeColor=<rgba 0.2656 1 0 1>}
  n18 = Multiply(A=VertexColor().R) {constB=3.7588}
  n19 = Time()
  NamedRerouteDeclaration(None=Add(A=Add(A=@n18, B=@n19), B=ParticleRandom())) {name=StripTime, nodeColor=<rgba 0.5703 0 1 1>}
  n20 = Transform(None=VectorParameter() {parameterName=Axis, defaultValue=<rgba 1 0 0 1>}.RGB) {transformSourceType=TRANSFORMSOURCE_Local}
  NamedRerouteDeclaration(None=RotateAboutAxis(
    NormalizedRotationAxis=Normalize(VectorInput=@n20)
    RotationAngle=ScalarParameter() {parameterName=Angle}
    PivotPoint=MaterialFunctionCall() {materialFunction=&ObjectPivotPoint}.Object Pivot Location
    Position=WorldPosition().XYZ
   )) {name=Rotated, nodeColor=<rgba 1 0.5625 0 1>}
  n21 = NamedRerouteUsage()
  n22 = DepthFade(Opacity=@n21, FadeDistance=ScalarParameter() {defaultValue=2, parameterName=DepthFade})
  QualitySwitch(Default=@n22, Low=@n21, High=@n22, Medium=@n21, Epic=@n22)
  Time()
  VertexColor()
 ]
}
