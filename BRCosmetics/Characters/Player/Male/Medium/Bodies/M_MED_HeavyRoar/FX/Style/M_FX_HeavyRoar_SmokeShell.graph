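# Translucent, unlit, two-sided smoke shell material for Niagara mesh particles in the HeavyRoar character's effects.
# A panning, axis-swizzled soft smoke trail texture is added to a slower-panning cloud texture to form the smoke pattern.
# A UV V-gradient times vertex red, computed in a vertex interpolator, masks that pattern, which an Opacity parameter then scales.
# Particle color RGB feeds the emissive through a shared emissive adjustment function.
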Material /BRCosmetics/Characters/Player/Male/Medium/Bodies/M_MED_HeavyRoar/FX/Style/M_FX_HeavyRoar_SmokeShell
graph {
 properties {
  bUsedWithNiagaraMeshParticles=true
  blendMode=BLEND_Translucent
  pixelDepthOffsetMode=PDOM_Legacy
  shadingModel=MSM_Unlit
  twoSided=true
  usedShadingModels=Unlit
 }
 nodes [
  MaterialFunctionCall(Emissive=ParticleColor().RGB) {materialFunction=&MF_FX_EmissiveAdjustments}
  n1 = TextureCoordinate() {uTiling=5, vTiling=0.5}
  n2 = TextureSample(UVs=MaterialFunctionCall(XY=Panner(Coordinate=@n1) {speedY=0.5}) {materialFunction=&Swizzle}.YX) {
   samplerSource=SSM_Wrap_WorldGroupSettings
   samplerType=SAMPLERTYPE_Grayscale
   texture=&T_Smoke_Trail_Soft
  }
  n3 = TextureCoordinate() {uTiling=4, vTiling=0.25}
  n4 = Add(A=@n2.G, B=TextureSample(UVs=Panner(Coordinate=@n3) {speedY=0.2}) {texture=/Game/Effects/Fort_Effects/Textures/Smoke/T_Clouds_02_M}.R)
  n5 = ComponentMask(None=TextureCoordinate() {vTiling=6}) {g=true}
  n6 = VertexInterpolator(VS=Multiply(A=Saturate(None=@n5), B=VertexColor().R))
  Multiply(A=ScalarParameter() {defaultValue=0.7, parameterName=Opacity}, B=Saturate(None=Multiply(A=@n4, B=@n6.PS)))
 ]
}
