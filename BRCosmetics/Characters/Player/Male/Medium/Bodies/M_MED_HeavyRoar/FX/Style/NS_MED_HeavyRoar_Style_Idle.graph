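# Idle style Niagara effect for the HeavyRoar character cosmetic, adding dark smoke and glowing, lit eyes.
# Body_Smoke meshes spawn between sampled bone pairs, while Hoodsmoke, Wrists and EyeGlows follow specific bones.
# Smoke emitters feed time-smoothed positions to dynamic material parameters for lagging drag, adjusted while skydiving.
# A kill-reaction blend drives eye glow and body smoke, and visibility tags respond to hidden head or hand flags.
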NiagaraSystem /BRCosmetics/Characters/Player/Male/Medium/Bodies/M_MED_HeavyRoar/FX/Style/NS_MED_HeavyRoar_Style_Idle
dependencies {
 usedDataInterfaces [
  &NiagaraDataInterfaceArrayFloat
  &NiagaraDataInterfaceCurve
  &NiagaraDataInterfaceArrayFloat3
  &NiagaraDataInterfaceColorCurve
  &NiagaraDataInterfaceMeshRendererInfo
  &NiagaraDataInterfaceArrayInt32
  &NiagaraDataInterfaceSkeletalMesh
  &NiagaraDataInterfaceArrayFloat2
 ]
 usedDynamicInputs []
 usedModules [
  &self:SystemSpawnScript.SystemScriptSource.SystemScriptGraph.NiagaraNodeAssignment_1.SetVariables_#1
  &SystemState
  &Timeline
  &self:SystemSpawnScript.SystemScriptSource.SystemScriptGraph.NiagaraNodeAssignment_0.SetVariables_#2
 ]
}
emitters {
 Body_Smoke {
  emitterUpdateScript [
   &EmitterState(Life Cycle Mode=::NewEnumerator1, Loop Duration=1, UseLoopDelay=false)
   &SpawnRate(SpawnRate=9, Use Spawn Probability=false)
  ]
  particleSpawnScript [
   &InitializeParticle(
    Color=<rgba 0.01 0.01 0.01 1>
    Lifetime=&RandomRangeFloat(Minimum=1, Maximum=2.5)
    Mesh Scale=<1.5 1.5 4>
    Mesh Scale Mode=::NewEnumerator4
    Sprite Size=<20 40>
    Sprite Size Mode=::NewEnumerator4
    UsePositionOffset=false
    Write Material Random=true
   )
   &FN_SetSpawnGroup(Number of Spawn Groups=6)
   /SetVariables() {isSetParameters=true}
   &CameraOffset(Camera Offset Amount=&MakeCustomFloatFromBool(
     Bool=&SetBoolByIntComparison(A=&SpawnGroup(), B=3, Comparison Type=::NewEnumerator4)
     True Float=-10
     False Float=16
    ))
   /SetVariables(Particles.MeshOrientation=&MakeQuatFromAxisAngle(Axis=<0 0 1>, AngleInDegrees=&RandomRangeFloat(Minimum=0, Maximum=180), NormalizeAxis=true)) {isSetParameters=true}
  ]
  particleUpdateScript [
   &ParticleState()
   &ScaleMeshSize(Scale Factor=&MakeVector(
     X=1
     Y=1
     Z=&SelectFloatFromArray(
      Direct Array Index=&SpawnGroup()
      Float Selection Array=DataInterface() {propertyValues {FloatData [1.1,2,1.1,1.1,2,1.1,1]}}
     )
    ))
   /SetVariables(Particles.Selection:NiagaraFloat=&RandomRangeFloat(Minimum=0, Maximum=1)) {isSetParameters=true}
   &SkeletalMeshLocation(
    Bone / Socket Index=$Particles.SpawnGroup
    Bone Sampling Mode=::NewEnumerator4
    "Interpolate Mesh Sampling (Update)"=true
    MaskSpawnGroup=false
    Orientation Sampling=::NewEnumerator1
    Sampled Position Coordinate Space=::World
    Sampled Position Offset Coordinate Space:ENiagaraSkelMeshTransforms=::NewEnumerator1
    Skeletal Mesh=$System.Skeletal Mesh
    Write To Particle Position=false
   ) {module=SkeletalMeshLocation001}
   &SkeletalMeshLocation(
    Bone / Socket Index=&Add_Integer(A=$Particles.Initial.SpawnGroup, B=1)
    Bone Sampling Mode=::NewEnumerator4
    "Interpolate Mesh Sampling (Update)"=true
    MaskSpawnGroup=false
    Orientation Sampling=::NewEnumerator1
    Sampled Position Coordinate Space=::World
    Skeletal Mesh=$System.Skeletal Mesh
    Write To Particle Position=false
   ) {module=SkeletalMeshLocation002}
   /SetVariables(Particles.Position=&LerpPosition(A=$Particles.SkeletalMeshLocation001.SampledPosition, B=$Particles.SkeletalMeshLocation002.SampledPosition, Alpha=$Particles.Selection)) {isSetParameters=true}
   /SetVariables(Particles.Center size:Vector2f=&Multiply_Vector2D_ByFloat(Float=2, Vector2D=<85 160>)) {isSetParameters=true}
   &ScaleColor(
    Scale Alpha=&Multiply_Float(
     A=&FN_FrontEndFloatModifier(Gameplay=1, Front End=1)
     B=&FloatFromCurve(FloatCurve=DataInterface() {propertyValues {Curve {keys [
          <at 0 0 RCIM_Cubic>
          <at 0.7 1 RCIM_Cubic>
          {arriveTangent=0.5326, interpMode=RCIM_Cubic, leaveTangent=0.5326, tangentMode=RCTM_User, time=1, value=0}
         ]}}})
    )
    Scale RGB=&Multiply_VectorByFloat(
     Vector=&Multiply_VectorByFloat(Vector=<1 1 1>, Float=&FN_FrontEndFloatModifier(Gameplay=2, Front End=1))
     Float=&FNTimeOfDayScalar(Night Scale=0.3, Morning Scale=0.3, Evening Scale=0.4)
    )
    ScaleA=true
    ScaleRGB=false
   )
   &ScaleColor(
    Color Value To Scale=$StackContext.Color
    Scale Alpha=&Lerp_Float(A=0.5, B=1, Alpha=$System.KillReactBlend)
    ScaleA=true
    ScaleRGB=false
   ) {module=ScaleColor001}
   &InterpolateOverTime(
    Rate Of Change:NiagaraFloat=&RandomRangeFloat(Minimum=5, Maximum=15)
    Target Vector Value:Vector3f=&ConvertPositionToVector(Input Position=$Particles.Position, Convert To Absolute Large World Vector=false)
    Variable Type=::NewEnumerator2
   )
   /SetVariables(Particles.SpriteDynParams:Vector4f=&MakeVector4(
     X=0
     Y=&FloatFromCurve(
      FloatCurve=DataInterface() {propertyValues {Curve {keys [{tangentMode=RCTM_User, time=0, value=0}, {tangentMode=RCTM_User, time=1, value=0.5}]}}}
      Scale Curve=&FloatFromCurve(FloatCurve=DataInterface() {propertyValues {Curve {keys [<at 0 1.7 RCIM_Cubic>,<at 1 1.7 RCIM_Cubic>]}}})
     )
     Z=0
     W=0
    )) {isSetParameters=true}
   /SetVariables(Particles.Interpolate Dynamic=&MakeFloat4FromVectorAndFloat("Vector (RGB)"=$StackContext.InterpolateOverTime.InterpolateOverTime_V3_.Moving Average, "Float (Alpha)"=1)) {isSetParameters=true}
   /SetVariables(Particles.InterpolatedPosition:NiagaraPosition=$Particles.InterpolateOverTime.InterpolateOverTime_V3_.Moving Average) {isSetParameters=true}
   &DynamicMaterialParameters(
    Index 0 Param 1=&RandomRangeFloat(Minimum=0.9, Maximum=0.75)
    Index 0 Param 2=&RandomRangeFloat(Minimum=0.58, Maximum=1)
    Index 0 Param 3=50
    Index 0 Param 4=1
    Index 1 Param 1=&MakeFloatFromVector4(Vector 4=$Particles.Interpolate Dynamic, Channel=::NewEnumerator0)
    Index 1 Param 2=&MakeFloatFromVector4(Vector 4=$Particles.Interpolate Dynamic, Channel=::NewEnumerator1)
    Index 1 Param 3=&Lerp_Float(
     A=&MakeFloatFromVector(Vector=Linked() {linkedVariable {name=Particles.Position, type=NiagaraPosition}}, Channel=::NewEnumerator2)
     B=&MakeFloatFromVector4(Vector 4=$Particles.Interpolate Dynamic, Channel=::NewEnumerator2)
     Alpha=&MakeCustomFloatFromBool(Bool=$User.IsSkydiving, True Float=0.1, False Float=0.35)
    )
    Index 1 Param 4=1.5
    Index 2 Param 1=&Lerp_Float(A=6, B=8, Alpha=$System.KillReactBlend)
    Index 2 Param 2=&Lerp_Float(A=0.2, B=0.45, Alpha=$System.KillReactBlend)
    Module.Param0WriteEnabled=true
    Module.Param0WriteEnabled1=true
    Module.Param0WriteEnabled2=true
    Module.Param1WriteEnabled=true
    Module.Param1WriteEnabled1=true
    Module.Param1WriteEnabled2=true
    Module.Param2WriteEnabled=true
    Module.Param2WriteEnabled1=true
    Module.Param2WriteEnabled2=true
    Module.Param3WriteEnabled=true
    Module.Param3WriteEnabled1=true
    Module.Param3WriteEnabled2=true
    Write Parameter Index 1=true
    Write Parameter Index 2=true
   ) {module=DynamicMaterialParameters001}
   /SetVariables(Particles.SmokeStripColor:LinearColor=&MakeLinearColorFromVectorAndFloat(
     "Vector (RGB)"=<1 1 1>
     "Float (Alpha)"=&FloatFromCurve(
      FloatCurve=DataInterface() {propertyValues {Curve {keys [
          {interpMode=RCIM_Cubic, tangentMode=RCTM_User, time=0, value=0}
          {interpMode=RCIM_Cubic, tangentMode=RCTM_User, time=0.5, value=1}
          {tangentMode=RCTM_User, time=1, value=0}
         ]}}}
      Scale Curve=0.2
     )
    )) {isSetParameters=true}
   &SpriteFacingAndAlignment(Sprite Alignment=<0 0 1>, WriteAlignment=true, WriteFacing=true)
   /SetVariables(Particles.VisibilityTag=&self:PartHiddenCheck(
     Hand Hidden:NiagaraBool=$User.bisHandHidden
     Head HIdden:NiagaraBool=$User.bisHeadHidden
     Selection:NiagaraFloat=$Particles.Selection
     Spawn Group:NiagaraInt32=$Particles.SpawnGroup
    )) {isSetParameters=true}
  ]
  rendererClasses [&NiagaraMeshRendererProperties]
  renderers [{rendererClass=&NiagaraMeshRendererProperties, rendererIndex=0}]
 }
 EyeGlows {
  emitterUpdateScript [
   &EmitterState(Life Cycle Mode=::NewEnumerator1, Loop Behavior=::NewEnumerator1, Loop Duration Mode=::NewEnumerator1, UseLoopDelay=false)
   &SpawnBurst_Instantaneous(Use Spawn Probability=false, UseLoopCountLimit=false)
   &SpawnBurst_Instantaneous(Spawn Group=1, Use Spawn Probability=false, UseLoopCountLimit=false) {module=SpawnBurst_Instantaneous001}
  ]
  particleSpawnScript [&InitializeParticle(
    Lifetime=$Emitter.CurrentLoopDuration
    Sprite Size Mode=::NewEnumerator3
    Uniform Sprite Size=15
    UsePositionOffset=false
    Write Material Random=true
   )]
  particleUpdateScript [
   &ParticleState(Kill Particles When Lifetime Has Elapsed=false, Let Infinitely Lived Particles Die When Emitter Deactivates=true)
   &ScaleSpriteSize(
    Uniform Curve Index=$System.KillReactBlend
    Uniform Curve Sprite Scale=DataInterface() {propertyValues {Curve {keys [<at 0 1 RCIM_Cubic>,<at 1 1.25 RCIM_Cubic>]}}}
   )
   &SkeletalMeshLocation(
    Bone / Socket Index=3
    Bone Sampling Mode=::NewEnumerator4
    MaskSpawnGroup=false
    Sampled Position Offset=&Multiply_Vector(
     A=<4 10 2.75>
     B=&SelectVectorFromArray(
      Direct Array Index=&SpawnGroup()
      Vector Selection Array=DataInterface() {propertyValues {FloatData [<1 1 1>, <1 1 -1>]}}
     )
    )
    Skeletal Mesh=$System.Skeletal Mesh
   )
   &SolveForcesAndVelocity(Clamp Velocity=false, Limit Acceleration=false)
   &CameraOffset(Camera Offset Amount=4)
   &ScaleColor(
    Curve Index=$System.KillReactBlend
    Linear Color Curve=DataInterface() {propertyValues {
      bOptimizeLUT=false
      AlphaCurve {keys [<at 0 0.3>,<at 0.9607 0.7>]}
      BlueCurve {keys [<at 0 0.0281>,<at 0.9636 0>]}
      GreenCurve {keys [<at 0 0.3425>,<at 0.9636 0.0682>]}
      RedCurve {keys [<at 0 1>,<at 0.9636 2>]}
     }}
    Scale Mode=::NewEnumerator2
   )
   &FN_Gen5_Light(Color:LinearColor=<rgba 0.1256 1 0 1>, Position Offset:Vector3f=<20 0 0>)
   /SetVariables(Particles.LightColor=<rgba 250 80.3752 0 1>) {isSetParameters=true}
   /SetVariables(Particles.LightPosition:NiagaraPosition=&AddVectorToPosition(Position=$Particles.Position, Vector=&TransformVector(Vector=<-2 -2 0>, Destination Space=::World))) {isSetParameters=true}
   &ScaleColor(
    Color Value To Scale=$StackContext.Color
    Scale Alpha=&FloatFromCurve(
     FloatCurve=DataInterface() {propertyValues {Curve {keys [{interpMode=RCIM_Cubic, tangentMode=RCTM_User, time=0, value=0}, {interpMode=RCIM_Cubic, tangentMode=RCTM_User, time=1, value=1}]}}}
     CurveIndex=$System.Age
    )
    ScaleA=true
    ScaleRGB=true
   ) {module=ScaleColor001}
   /SetVariables(Particles.VisibilityTag=&MakeCustomIntFromBool(Bool=$User.bisHeadHidden, True INT=1, False INT=0)) {isSetParameters=true}
  ]
  rendererClasses [&NiagaraSpriteRendererProperties, &NiagaraLightRendererProperties]
  renderers [
   {rendererClass=&NiagaraSpriteRendererProperties}
   {rendererClass=&NiagaraSpriteRendererProperties, rendererIndex=1}
   {rendererClass=&NiagaraLightRendererProperties, rendererIndex=2}
  ]
 }
 Hoodsmoke {
  emitterUpdateScript [
   &EmitterState(Life Cycle Mode=::NewEnumerator1, Loop Behavior=::NewEnumerator1, Loop Duration Mode=::NewEnumerator1, UseLoopDelay=false)
   &SpawnBurst_Instantaneous(Spawn Group=2, Use Spawn Probability=false, UseLoopCountLimit=false) {module=SpawnBurst_Instantaneous002}
  ]
  particleSpawnScript [
   &InitializeParticle(
    Color=<rgba 0 0 0 0.9>
    Lifetime=$Emitter.CurrentLoopDuration
    Mesh Scale Mode=::NewEnumerator3
    Sprite Size Mode=::NewEnumerator3
    Uniform Sprite Size=35
    UsePositionOffset=false
    Write Material Random=true
   )
   /SetVariables() {isSetParameters=true, module=SetVariables_#3}
  ]
  particleUpdateScript [
   &ParticleState(Kill Particles When Lifetime Has Elapsed=false, Loop Particles Lifetime=true)
   &SkeletalMeshLocation(
    Bone / Socket Index=&SelectIntFromArray(Direct Array Index=&SpawnGroup(), Int Selection Array=DataInterface() {propertyValues {IntData [0,7,3]}})
    Bone Sampling Mode=::NewEnumerator4
    MaskSpawnGroup=false
    Orientation Sampling=::NewEnumerator1
    Sampled Orientation Coordinate Space=::World
    Skeletal Mesh=$System.Skeletal Mesh
   )
   /SetVariables(Particles.MeshOrientation=&MultiplyQuaternion(
     Quaternion A=$StackContext.SkeletalMeshLocation.SampledRotation
     Quaternion B=&MakeQuatFromAxisAngle(
      Axis=<0 1 0>
      AngleInDegrees=&MakeCustomFloatFromBool(
       Bool=&SetBoolByIntComparison(A=&SpawnGroup(), B=0, Comparison Type=::NewEnumerator4)
       True Float=0
       False Float=180
      )
     )
    )) {isSetParameters=true, module=SetVariables_#4}
   /SetVariables(Particles.MeshOrientation=&OrientMeshToVector(Look At Direction=<0 0 1>, Reference Axis Vector=<0 0 1>)) {enabled=false, isSetParameters=true, module=SetVariables_#5}
   &InterpolateOverTime(
    Initial Vector Value:NiagaraPosition=$Particles.Initial.Position
    Rate Of Change:NiagaraFloat=&MakeCustomFloatFromBool(Bool=$User.IsSkydiving, True Float=16, False Float=12.5)
    Target Vector Value:Vector3f=&ConvertPositionToVector(
     Input Position=&AddVectorToPosition(Position=$Particles.Position, Vector=<0 0 10>)
     Convert To Absolute Large World Vector=false
    )
    Variable Type=::NewEnumerator2
   )
   /SetVariables(Particles.DragPosition:NiagaraPosition=$StackContext.InterpolateOverTime.InterpolateOverTime_V3_.Moving Average) {isSetParameters=true, module=SetVariables_#6}
   &ScaleMeshSize(Scale Factor=&MakeVector(
     X=&MakeCustomFloatFromBool(
      Bool=&SetBoolByIntComparison(A=&SpawnGroup(), B=2, Comparison Type=::NewEnumerator6)
      True Float=&RemapRange(
       Input Value To Remap=&VectorLength(Vector=&Multiply_Vector(A=$Engine.Owner.Velocity, B=<1 1 0>))
       Input Min=100
       Input Max=400
       Output Min=0.7
       Output Max=1
      )
      False Float=1
     )
     Y=1
     Z=1
    ))
   &SolveForcesAndVelocity(Clamp Velocity=false, Limit Acceleration=false)
   &DynamicMaterialParameters(
    Index 0 Param 1=&MakeFloatFromVector(Vector=Linked() {linkedVariable {name=Particles.DragPosition, type=NiagaraPosition}}, Channel=::NewEnumerator0)
    Index 0 Param 2=&MakeFloatFromVector(Vector=Linked() {linkedVariable {name=Particles.DragPosition, type=NiagaraPosition}}, Channel=::NewEnumerator1)
    Index 0 Param 3=&Lerp_Float(
     A=&MakeFloatFromVector(Vector=Linked() {linkedVariable {name=Particles.DragPosition, type=NiagaraPosition}}, Channel=::NewEnumerator2)
     B=&MakeFloatFromVector(Vector=Linked() {linkedVariable {name=Particles.Position, type=NiagaraPosition}}, Channel=::NewEnumerator2)
     Alpha=0.75
    )
    Index 0 Param 4=1
    Module.Param0WriteEnabled=true
    Module.Param1WriteEnabled=true
    Module.Param2WriteEnabled=true
    Module.Param3WriteEnabled=true
   )
   /SetVariables(Particles.VisibilityTag=&MakeCustomIntFromBool(Bool=$User.bisHeadHidden, True INT=1, False INT=0)) {isSetParameters=true}
  ]
  rendererClasses [&NiagaraMeshRendererProperties, &NiagaraSpriteRendererProperties]
  renderers [
   {rendererClass=&NiagaraMeshRendererProperties, rendererIndex=0}
   {rendererClass=&NiagaraSpriteRendererProperties, rendererIndex=1}
  ]
 }
 Wrists {
  emitterUpdateScript [
   &EmitterState(Life Cycle Mode=::NewEnumerator1, Loop Behavior=::NewEnumerator1, Loop Duration Mode=::NewEnumerator1, UseLoopDelay=false)
   &SpawnBurst_Instantaneous(Spawn Count=3, Use Spawn Probability=false, UseLoopCountLimit=false)
   &SpawnBurst_Instantaneous(Spawn Count=3, Spawn Group=1, Use Spawn Probability=false, UseLoopCountLimit=false) {module=SpawnBurst_Instantaneous001}
  ]
  particleSpawnScript [
   &InitializeParticle(
    Color=<rgba 0 0 0 0.85>
    Lifetime=$Emitter.CurrentLoopDuration
    Mesh Scale=&MakeVector(X=&RandomRangeFloat(Minimum=0.75, Maximum=1.1), Y=0.75, Z=0.75)
    Mesh Scale Mode=::NewEnumerator4
    Sprite Size Mode=::NewEnumerator3
    Uniform Sprite Size=35
    UsePositionOffset=false
    Write Material Random=true
   )
   /SetVariables() {isSetParameters=true, module=SetVariables_#3}
  ]
  particleUpdateScript [
   &ParticleState(Kill Particles When Lifetime Has Elapsed=false, Loop Particles Lifetime=true)
   &SkeletalMeshLocation(
    Bone / Socket Index=&SelectIntFromArray(Direct Array Index=&SpawnGroup(), Int Selection Array=DataInterface() {propertyValues {IntData [0,7,3]}})
    Bone Sampling Mode=::NewEnumerator4
    MaskSpawnGroup=false
    Orientation Sampling=::NewEnumerator1
    Sampled Orientation Coordinate Space=::World
    Skeletal Mesh=$System.Skeletal Mesh
   )
   &InterpolateOverTime(
    Initial Vector Value:NiagaraPosition=$Particles.Initial.Position
    Rate Of Change:NiagaraFloat=&MakeCustomFloatFromBool(Bool=$User.IsSkydiving, True Float=10, False Float=8)
    Target Vector Value:Vector3f=&ConvertPositionToVector(
     Input Position=&AddVectorToPosition(Position=$Particles.Position, Vector=<0 0 10>)
     Convert To Absolute Large World Vector=false
    )
    Variable Type=::NewEnumerator2
   )
   /SetVariables(Particles.DragPosition:NiagaraPosition=$StackContext.InterpolateOverTime.InterpolateOverTime_V3_.Moving Average) {isSetParameters=true, module=SetVariables_#6}
   /SetVariables(Particles.MeshOrientation=&OrientMeshToVector(Look At Direction=<0 0 1>, Reference Axis Vector=<-1 0 0>)) {isSetParameters=true, module=SetVariables_#5}
   /SetVariables(Particles.MeshOrientation=&MultiplyQuaternion(
     Quaternion A=$StackContext.MeshOrientation
     Quaternion B=&MakeQuatFromAxisAngle(
      Axis=<1 0 0>
      AngleInDegrees=&Multiply_Float(A=&ReturnNormalizedExecIndex(), B=720)
     )
    )) {isSetParameters=true, module=SetVariables_#4}
   &ScaleMeshSize(Scale Factor=&MakeVector(
     X=&MakeCustomFloatFromBool(
      Bool=&SetBoolByIntComparison(A=&SpawnGroup(), B=2, Comparison Type=::NewEnumerator6)
      True Float=&RemapRange(
       Input Value To Remap=&VectorLength(Vector=&Multiply_Vector(A=$Engine.Owner.Velocity, B=<1 1 0>))
       Input Min=100
       Input Max=400
       Output Min=0.7
       Output Max=1
      )
      False Float=1
     )
     Y=1
     Z=1
    ))
   &SolveForcesAndVelocity(Clamp Velocity=false, Limit Acceleration=false)
   &DynamicMaterialParameters(
    Index 0 Param 1=&MakeFloatFromVector(Vector=Linked() {linkedVariable {name=Particles.DragPosition, type=NiagaraPosition}}, Channel=::NewEnumerator0)
    Index 0 Param 2=&MakeFloatFromVector(Vector=Linked() {linkedVariable {name=Particles.DragPosition, type=NiagaraPosition}}, Channel=::NewEnumerator1)
    Index 0 Param 3=&Lerp_Float(
     A=&MakeFloatFromVector(Vector=Linked() {linkedVariable {name=Particles.DragPosition, type=NiagaraPosition}}, Channel=::NewEnumerator2)
     B=&MakeFloatFromVector(Vector=Linked() {linkedVariable {name=Particles.Position, type=NiagaraPosition}}, Channel=::NewEnumerator2)
     Alpha=0.75
    )
    Index 0 Param 4=1
    Module.Param0WriteEnabled=true
    Module.Param1WriteEnabled=true
    Module.Param2WriteEnabled=true
    Module.Param3WriteEnabled=true
   )
   &ScaleColor(Scale Alpha=$System.Timeline001.InverseNormalizedTime, ScaleA=true, ScaleRGB=false)
   /SetVariables(Particles.VisibilityTag=&MakeCustomIntFromBool(Bool=$User.bisHandHidden, True INT=1, False INT=0)) {isSetParameters=true}
  ]
  rendererClasses [&NiagaraMeshRendererProperties]
  renderers [{rendererClass=&NiagaraMeshRendererProperties, rendererIndex=0}]
 }
}
summary {
 systemName=NS_MED_HeavyRoar_Style_Idle
 emitters [
  {emitterName=Body_Smoke, rendererClasses [&NiagaraMeshRendererProperties]}
  {emitterName=EyeGlows, rendererClasses [&NiagaraSpriteRendererProperties, &NiagaraLightRendererProperties]}
  {emitterName=Hoodsmoke, rendererClasses [&NiagaraMeshRendererProperties, &NiagaraSpriteRendererProperties]}
  {emitterName=Wrists, rendererClasses [&NiagaraMeshRendererProperties]}
 ]
 userVariables [
  User.bIsFrontEnd {type=NiagaraBool}
  User.bIsFrontEndPreview {type=NiagaraBool}
  User.bisHandHidden {type=NiagaraBool}
  User.bisHeadHidden {type=NiagaraBool}
  User.bKillReaction {type=NiagaraBool}
  User.global Alpha {type=NiagaraFloat, defaultValue=0, description=""}
  User.IsSkydiving {type=NiagaraBool}
 ]
}
